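# Conversation database for the Weeping Woods NPC, Bushy, providing its wasteland-themed greeting dialogue.
# The entry point leads to a single speech node that has a default line and contextual variants.
# A first-ever-conversation greeting uses a curve-table priority row, while the other lines are gated by this NPC's match conversation count.
# The speech then branches through link nodes to the shared NPC library quests and basic services conversations.
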ConversationDatabase /NPCLibrary/NPCs/WeepingWoods/NPCLibrary_Conversation_WeepingWoods
entry_points [{tag=Conversation.Entry.WeepingWoods, destinations [#1]}]
nodes [
 ConversationEntryPointNode_7 {
  id=#1
  class=ConversationEntryPointNode
  outputs [#2]
  properties {entryTag {tagName=Conversation.Entry.WeepingWoods}}
 }
 Speech_C_4 {
  id=#2
  class=Speech_C
  outputs [#3, #4]
  properties {
   nodeName=""
   generalConfig {
    defaultMessage="Wasteland full of scarybad and bigboom, but Bushy is ready!"
    contextualMessages [
     {
      message="Greetingsgreetings, friend!"
      contextRequirements [{requirement {class=FortControllerRequirement_IsFirstEverConversationWithNPC}}]
      priority {curve {rowName=Tandem.Speech.Priority.FirstEverInteraction}}
     }
     {
      message="Wasteland full of scarybad and bigboom, but Bushy is ready!"
      contextRequirements [{requirement {class=FortControllerRequirement_CheckNPCMatchConversationCount, properties {nPCUniqueIdTag {tagName=AISpawnerData.Type.Tandem.WeepingWoods}}}}]
      priority {curve {rowName=None}}
     }
     {
      message="Wastelander Bushy ready for explosions and adventure! But not too much."
      contextRequirements [{requirement {class=FortControllerRequirement_CheckNPCMatchConversationCount, properties {nPCUniqueIdTag {tagName=AISpawnerData.Type.Tandem.WeepingWoods}}}}]
      priority {curve {rowName=None}}
     }
     {
      message=Bushy not a fan of big dustrydry wasteland. Bushy look for water soon...
      contextRequirements [{requirement {class=FortControllerRequirement_CheckNPCMatchConversationCount, properties {nPCUniqueIdTag {tagName=AISpawnerData.Type.Tandem.WeepingWoods}}}}]
      priority {curve {rowName=None}}
     }
     {
      message=Bushy branches feel a little stiffsnap. Wasteland windydry bad for bark.
      contextRequirements [{requirement {class=FortControllerRequirement_CheckNPCMatchConversationCount, properties {nPCUniqueIdTag {tagName=AISpawnerData.Type.Tandem.WeepingWoods}}}}]
      priority {curve {rowName=None}}
     }
    ]
   }
  }
 }
 ConversationLinkNode_0 {id=#4, class=ConversationLinkNode, properties {remoteEntryTag {tagName=Conversation.Entry.BasicServices}}}
 ConversationLinkNode_1 {id=#3, class=ConversationLinkNode, properties {remoteEntryTag {tagName=Conversation.Entry.NPCLibrary.Quests}}}
]
